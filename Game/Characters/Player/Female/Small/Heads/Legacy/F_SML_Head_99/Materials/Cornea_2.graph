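# Masked, Default Lit cornea material for legacy small female player head 99, set up for skeletal meshes with morph targets and mesh deformers.
# It calls the shared CharacterResInEffect function, which by its name handles the character res-in effect.
# Separately, a grey-default vector parameter's RGB is scaled down by a constant.
# The graph shows no link from either branch to a material output.
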Material /Game/Characters/Player/Female/Small/Heads/Legacy/F_SML_Head_99/Materials/Cornea_2
graph {
 properties {
  bUsedWithMeshDeformer=true
  bUsedWithMorphTargets=true
  bUsedWithSkeletalMesh=true
  blendMode=BLEND_Masked
  pixelDepthOffsetMode=PDOM_Legacy
  usedShadingModels=Default Lit
 }
 nodes [
  MaterialFunctionCall!(Animation Gradient, Rez In Percentage) {materialFunction=/Game/Packages/Fortress_SharedMaterials/Base_Material_Functions/CharacterResInEffect}
  Multiply(A=VectorParameter() {defaultValue=<rgba 0.5 0.5 0.5 0>}.RGB, B=Constant() {r=0.2})
 ]
}
